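; Looping gameplay cue for a CleanSweepDiagram EMP impact that plays a screen glitch when the affected Fortnite player pawn is locally viewed.
; The construction script creates a dynamic instance of the EMP grenade glitch material and adds it as a blendable to the cue's post-process component.
; On loop start, it enables that component, and a timeline drives the material's OverallEffectIntensity, disabling the component when the timeline finishes.
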(event EventOnLoopingStart(EitherVFX-preferthisversion)
  (params (MyTarget :type Actor%20Object%20Reference) (Parameters :type Gameplay%20Cue%20Parameters%20Structure%20%28by%20ref%29) (ParticleComponents :type Array%20of%20FXSystem%20Component%20Object%20References) (AudioComponents :type Array%20of%20Audio%20Component%20Object%20References))
  ;[1 1 1] On Loop Start
  (|CallParentFunction MyTarget Parameters ParticleComponents AudioComponents)
  (bind _asfort_player_pawn (Utilities|Casting|CastToFortPlayerPawn MyTarget))
  (if (Pawn|IsLocallyViewed _asfort_player_pawn)
    (CallFunction|PlayGlitch)))

(event Custom|PlayGlitch
  (Variables|/Script/Engine.PostProcessComponent|SetbEnabled true (Variables|Default|GetPostProcess))
  (bind (_direction _pp) (Timeline|Timeline PlayFromStart)
    (:Update
      (Rendering|Material|SetScalarParameterValue (Variables|Default|GetPP_M_Instance) "OverallEffectIntensity" _pp :!class /Script/Engine.MaterialInstanceDynamic))
    (:Finished
      (Variables|/Script/Engine.PostProcessComponent|SetbEnabled false (Variables|Default|GetPostProcess)))))

(fn ConstructionScript
  (bind _returnvalue (Rendering|Material|CreateDynamicMaterialInstance 0 "/EMPGameplay/Items/EMPGrenade/PostProcessing/MI_PP_EmpGrenade_Glitch_Effect.MI_PP_EmpGrenade_Glitch_Effect" "None" "None" :!class /Script/Engine.KismetMaterialLibrary))
  (bind _output_get (Variables|Default|SetPP_M_Instance _returnvalue))
  (Rendering|AddorUpdateBlendable (Variables|Default|GetPostProcess) _output_get 1.0 :!class /Script/Engine.PostProcessComponent))
