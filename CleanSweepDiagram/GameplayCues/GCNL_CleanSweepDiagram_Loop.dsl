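; Looping gameplay cue for the CleanSweepDiagram cloak effect on Fortnite player pawns, extending parent cloak-cue behavior.
; Loop start turns off the target's movement dust FX, and removal turns it back on.
; OnBurst adds a cloaked highlighting override to the instigator's pawn, found through its player state, and OnCeaseRelevant removes it by stored handle.
; After the parent applies cloak materials, a scalable float sets their stationary fill opacity.
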(event EventBeginPlay)

(event EventOnLoopingStart(EitherVFX-preferthisversion)
  (params (MyTarget :type Actor%20Object%20Reference) (Parameters :type Gameplay%20Cue%20Parameters%20Structure%20%28by%20ref%29) (ParticleComponents :type Array%20of%20FXSystem%20Component%20Object%20References) (AudioComponents :type Array%20of%20Audio%20Component%20Object%20References))
  (|CallParentFunction MyTarget Parameters ParticleComponents AudioComponents)
  (bind _asfort_player_pawn_athena (Utilities|Casting|CastToFortPlayerPawnAthena MyTarget))
  (Variables|/Script/FortniteGame.FortPlayerPawn|SetbFXPlayDustOnMovement false _asfort_player_pawn_athena))

(event Collision|EventActorBeginOverlap
  (params (OtherActor :type Actor%20Object%20Reference)))

(event EventTick
  (params (DeltaSeconds :type Float%20%28single-precision%29))
  (|CallParentFunction DeltaSeconds) ; This node is disabled and will not be called.\nDrag off pins to build functionality.
)

(event EventOnRemoval(EitherVFX-preferthisversion)
  (params (MyTarget :type Actor%20Object%20Reference) (Parameters :type Gameplay%20Cue%20Parameters%20Structure%20%28by%20ref%29) (ParticleComponents :type Array%20of%20FXSystem%20Component%20Object%20References) (AudioComponents :type Array%20of%20Audio%20Component%20Object%20References) (BurstCameraShakeInstance :type Camera%20Shake%20Base%20Object%20Reference) (BurstDecalInstance :type Decal%20Actor%20Object%20Reference))
  (|CallParentFunction MyTarget Parameters ParticleComponents AudioComponents BurstCameraShakeInstance BurstDecalInstance)
  (bind _asfort_player_pawn_athena (Utilities|Casting|CastToFortPlayerPawnAthena MyTarget))
  (Variables|/Script/FortniteGame.FortPlayerPawn|SetbFXPlayDustOnMovement true _asfort_player_pawn_athena))

(event EventApplyCloakMaterialOverride
  (|CallParentFunction)
  (Class|FortPlayerPawn|SetScalarParamonOverriddenMaterials (Variables|FortGameCueNotifyLoop_Cloak|GetTargetPlayer) "FillOpacityStationary" (ScalableFloat|GetValueatLevel (Variables|Default|GetHF_OpacityStationaryValue))))

(fn OnBurst
  (params (MyTarget :type Actor%20Object%20Reference) (Parameters :type Gameplay%20Cue%20Parameters%20Structure%20%28by%20ref%29))
  (bind _returnvalue (|CallParentFunction MyTarget Parameters))
  (bind _asfort_player_state (Utilities|Casting|CastToFortPlayerState (Ability|GameplayCue|GetInstigatorActor Parameters)))
  (bind _asfort_player_pawn_athena (Utilities|Casting|CastToFortPlayerPawnAthena (PlayerState|GetPawn _asfort_player_state)))
  (bind _returnvalue_2 (Highlighting|AddCustomHighlightingDataOverride _asfort_player_pawn_athena (Variables|Default|GetCloakedCustomHighlightingData) (Variables|Default|GetCloakingHighlightOverridePriority)))
  (Variables|Default|SetCloakedCustomHighlightingOverrideHandle _returnvalue_2)
  (return _returnvalue))

(fn OnCeaseRelevant
  (params (MyTarget :type Actor%20Object%20Reference) (Parameters :type Gameplay%20Cue%20Parameters%20Structure%20%28by%20ref%29))
  (bind _returnvalue (|CallParentFunction MyTarget Parameters))
  (bind _asfort_player_state (Utilities|Casting|CastToFortPlayerState (Ability|GameplayCue|GetInstigatorActor Parameters)))
  (bind _asfort_player_pawn_athena (Utilities|Casting|CastToFortPlayerPawnAthena (PlayerState|GetPawn _asfort_player_state)))
  (Highlighting|RemoveCustomHighlightingDataOverride _asfort_player_pawn_athena (Variables|Default|GetCloakedCustomHighlightingOverrideHandle))
  (return _returnvalue))

(fn ConstructionScript
  (|CallParentFunction))
