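; Gameplay cue burst for an EMP grenade impact on a player, showing shield-hit feedback.
; It reads total shields and damage from the cue magnitudes, then attaches shield-destroyed or shield-damaged Niagara to the chest, passing in the explosion location.
; For self-hits, it rebuilds shield damage or break effects and audio.
; Scalable-float toggles control damage numbers, including showing zero-damage hits as 1.
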(fn OnBurst(EitherVFX-preferthisversion)
  (params (MyTarget :type Actor%20Object%20Reference) (Parameters :type Gameplay%20Cue%20Parameters%20Structure%20%28by%20ref%29) (ParticleComponents :type Array%20of%20FXSystem%20Component%20Object%20References) (AudioComponents :type Array%20of%20Audio%20Component%20Object%20References) (BurstCameraShakeInstance :type Camera%20Shake%20Base%20Object%20Reference) (BurstDecalInstance :type Decal%20Actor%20Object%20Reference))
  (bind (_normalizedmagnitude _rawmagnitude _effectcontext _matchedtagname _originaltag _aggregatedsourcetags _aggregatedtargettags _location _normal _instigator) (Ability|GameplayCue|BreakGameplayCueParameters Parameters))
  (|CallParentFunction MyTarget Parameters ParticleComponents AudioComponents BurstCameraShakeInstance BurstDecalInstance)
  (LocalVariables|SetTarget MyTarget)
  (LocalVariables|SetTotalShields _normalizedmagnitude)
  (LocalVariables|SetDamage _rawmagnitude)
  (LocalVariables|SetGrenadeExplosionLocation _location)
  (LocalVariables|SetInstigator _instigator)
  (bind _returnvalue (Niagara|SpawnSystemAttached (select (<= (LocalVariables|GetTotalShields) (LocalVariables|GetDamage)) (Variables|Default|GetFX_ShieldDestroyed) (Variables|Default|GetFX_ShieldDamaged)) (Variables|/Script/Engine.Character|GetMesh (Utilities|Casting|CastToFortPlayerPawn (LocalVariables|GetTarget))) "FX_Chest" "0, 0, 0" (Math|Rotator|MakeRotator 0.0 0.0 (Math|Random|RandomFloatinRange 0.0 3.6e+02)) "KeepRelativeOffset" true true "AutoRelease" true))
  (Niagara|SetNiagaraVariable(Position) _returnvalue "OriginLocation" (LocalVariables|GetGrenadeExplosionLocation))
  ;[1 1 1] Damage Numbers
  (bind _asfort_player_state (Utilities|Casting|CastToFortPlayerState (LocalVariables|GetInstigator))
    (:then
      (bind _asplayer_pawn_athena (Utilities|Casting|CastToPlayerPawn_Athena (PlayerState|GetPlayerPawnorVehicleDriver _asfort_player_state))
        (:then
          (if (== (PlayerState|GetPlayerPawnorVehicleDriver _asfort_player_state) (LocalVariables|GetTarget))
            (if (ScalableFloat|GetValueasBool (Variables|Default|GetShowDamageNumbersInGC))
              (EventDispatchers|CallDisplayHitNotify :self _asplayer_pawn_athena :Damage (Math|Float|Max(Float) (LocalVariables|GetDamage) (Math|Float|SelectFloat 1.0 0.0 (ScalableFloat|GetValueasBool (Variables|Default|GetShowDamageNumbersAt0Damage)))) :Shield true :HitPawn (LocalVariables|GetTarget) :HitLocation (Transformation|GetActorLocation (LocalVariables|GetTarget))) ; Damage numbers
              ;[1 1 1] Shield Damage (rebuilt for instigator only, since it works correctly otherwise)
              (if (> (LocalVariables|GetTotalShields) (LocalVariables|GetDamage))
                (Audio|SpawnSoundAttached (AudioGameplay|GetAudioforDamageEvent (LocalVariables|GetTarget) (LocalVariables|GetInstigator) (Variables|Default|GetSound_ShieldDamaged_Receiver) (Variables|Default|GetSound_ShieldDamaged_3P) (Variables|Default|GetSound_ShieldDamaged_3P) (LocalVariables|GetTarget)) (Variables|/Script/Engine.Actor|GetRootComponent (LocalVariables|GetTarget)))
                (return)
                (else
                  _))
              (else
                ;[1 1 1] Shield Break FX (rebuilt) - play when dealing 0 damage or for instigator
                (if (not (<= (LocalVariables|GetDamage) 0))
                  (return))))
            (bind _asfort_pawn (Utilities|Casting|CastToFortPawn (LocalVariables|GetTarget))
              (:then
                (Niagara|SpawnSystemAttached (Variables|Default|GetFX_ShieldBreak_PawnFXDefault_LW) (Variables|/Script/Engine.Character|GetMesh _asfort_pawn) "FX_Chest" "0, 0, 0" "0, 0, 0" "KeepRelativeOffset" true)
                (Audio|SpawnSoundAttached (AudioGameplay|GetAudioforDamageEvent (LocalVariables|GetTarget) (LocalVariables|GetInstigator) (Variables|Default|GetSound_ShieldDestroyed_Receiver) (Variables|Default|GetSound_ShieldDestroyed_Instigator) (Variables|Default|GetSound_ShieldDestroyed_Observer) (LocalVariables|GetTarget)) (Variables|/Script/Engine.Actor|GetRootComponent (LocalVariables|GetTarget)))
                (return))
              (:CastFailed
                (return)))
            (elif (ScalableFloat|GetValueasBool (Variables|Default|GetShowDamageNumbersInGC))
              (if (and (<= (LocalVariables|GetDamage) 1.0) (ScalableFloat|GetValueasBool (Variables|Default|GetShowDamageNumbersAt0Damage)))
                (EventDispatchers|CallDisplayHitNotify :self _asplayer_pawn_athena :Damage 1.0 :Shield true :HitPawn (LocalVariables|GetTarget) :HitLocation (Transformation|GetActorLocation (LocalVariables|GetTarget))) ; Damage numbers
                (else
                  _))
              (else
                _))))
        (:CastFailed
          _)))
    (:CastFailed
      _)))
